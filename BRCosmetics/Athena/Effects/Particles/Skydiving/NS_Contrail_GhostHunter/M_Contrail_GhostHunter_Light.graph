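# Additive, unlit, two-sided Niagara mesh particle material for the light in the GhostHunter skydiving contrail.
# The block look tints screen-space scrolling scanlines and stepped noise with a curve-atlas gradient, masked by UV fades.
# Parameters blend this toward a simple UV mask or a hand hologram built from vertex color channels and curve-driven rings.
# Particle color scales the result, with optional fresnel fade and deferred-only depth fade outside hologram mode.
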Material /BRCosmetics/Athena/Effects/Particles/Skydiving/NS_Contrail_GhostHunter/M_Contrail_GhostHunter_Light
graph {
 properties {
  bUsedWithNiagaraMeshParticles=true
  blendMode=BLEND_Additive
  pixelDepthOffsetMode=PDOM_Legacy
  shadingModel=MSM_Unlit
  twoSided=true
  usedShadingModels=Unlit
 }
 nodes [
  n1 = ParticleColor()
  n2 = LinearInterpolate(A=NamedRerouteUsage(), B=NamedRerouteUsage(), Alpha=ScalarParameter() {group=SimpleMask, parameterName=UseSimpleMask})
  n3 = PixelNormalWS()
  n4 = VertexNormalWS()
  n5 = QualitySwitch(Default=@n3, Low=@n4, High=@n3, Medium=@n4, Epic=@n3)
  n6 = Abs(None=DotProduct(A=CameraVectorWS(), B=ShadingPathSwitch(Default=@n5, Deferred=@n5, Forward=@n4, Mobile=@n4)))
  n7 = OneMinus(None=@n6)
  n8 = VectorParameter() {
   group=Base
   parameterName=Mesh Fresnel Fade
   sortPriority=1
   channelNames {a="", b=Power, g=Invert, r=Use Mesh Fresnel Fade}
   defaultValue=<rgba 0 0 1 0>
  }
  n9 = Saturate(None=Power(Base=LinearInterpolate(A=@n6, B=@n7, Alpha=@n8.Invert), Exp=@n8.Power))
  n10 = Multiply(A=@n1.A, B=LinearInterpolate(A=@n2, B=Multiply(A=@n9, B=@n2), Alpha=@n8.Use Mesh Fresnel Fade))
  n11 = DepthFade(Opacity=@n10, FadeDistance=ScalarParameter() {defaultValue=10, group=Base, parameterName=DepthFade})
  LinearInterpolate(
   A=ShadingPathSwitch(Default=@n11, Deferred=@n11, Forward=@n10, Mobile=@n10)
   B=Multiply(A=@n1.A, B=NamedRerouteUsage())
   Alpha=ScalarParameter() {parameterName=Use HandHologram}
  )
  n12 = Multiply(A=NamedRerouteUsage(), B=@n1.RGB)
  n13 = ScalarParameter() {group=SimpleMask, parameterName=UseSimpleMask}
  n14 = NamedRerouteUsage()
  MaterialFunctionCall(Emissive=LinearInterpolate(A=LinearInterpolate(A=@n12, B=@n1.RGB, Alpha=@n13), B=Multiply(A=@n14, B=@n1.RGB), Alpha=ScalarParameter() {parameterName=Use HandHologram})) {materialFunction=&MF_FX_EmissiveAdjustments}
  n15 = CurveAtlasRowParameter(CurveTime=MaterialFunctionCall(Float2=TextureCoordinate()) {materialFunction=&BreakOutFloat2Components}.G) {
   atlas=/Game/Athena/LinearColorCurves/VFX_CurveAtlas
   curve=/Game/Athena/LinearColorCurves/Curve_Contrail_GhostHunter_02
   defaultValue=34
   parameterName=1
  }
  n16 = ScalarParameter() {group=Block, parameterName=Line Dark}
  n17 = ScalarParameter() {defaultValue=1, group=Block, parameterName=Line Light}
  n18 = MaterialFunctionCall(Float2=ScreenPosition().ViewportUV) {materialFunction=&BreakOutFloat2Components}
  n19 = ScalarParameter() {defaultValue=50, group=Block, parameterName=Line amount}
  n20 = Time()
  n21 = ScalarParameter() {defaultValue=0.2, group=Block, parameterName=Line Speed}
  n22 = Frac(None=Add(A=Multiply(A=@n18.G, B=@n19), B=Multiply(A=@n20, B=@n21)))
  n23 = OneMinus(None=@n22)
  n24 = Ceil(None=Add(A=Multiply(A=@n22, B=@n23), B=ScalarParameter() {defaultValue=-0.15, group=Block, parameterName=Line Size}))
  n25 = MaterialFunctionCall(Float2=TextureCoordinate()) {materialFunction=&BreakOutFloat2Components}
  n26 = Power(Base=@n25.G, Exp=ScalarParameter() {defaultValue=3, parameterName=BaseFade}) {constExponent=5}
  n27 = Multiply(A=ScreenPosition().ViewportUV, B=ScalarParameter() {defaultValue=10, group=Block, parameterName=BlockUV})
  n28 = DynamicParameter() {defaultValue=<rgba 0 0 1 1>}
  n29 = Multiply(A=Time(), B=ScalarParameter() {defaultValue=10.7749, group=Block, parameterName=Block U Speed}) {constB=0.5}
  n30 = TextureSample(UVs=Panner(Coordinate=Add(A=@n27, B=@n28.Param1), Time=Ceil(None=@n29)) {desc="-0.33,-0.43", speedX=-0.3, speedY=0.5}) {texture=/Game/Athena/UI/Challenges/Art/T_UI_DataChip_Noise}
  n31 = Add(
   A=Multiply(A=LinearInterpolate(A=@n16, B=@n17, Alpha=@n24), B=@n26)
   B=Multiply(A=LinearInterpolate(Alpha=@n30.R) {constA=-0.2, constB=2}, B=ScalarParameter() {defaultValue=1, parameterName=BlockMultiply})
  )
  n32 = ScalarParameter() {defaultValue=6, parameterName=LightFade}
  n33 = ScalarParameter() {defaultValue=10, parameterName=BattomFade}
  n34 = Multiply(A=Add(A=@n26, B=Power(Base=@n25.G, Exp=@n32) {constExponent=6}), B=OneMinus(None=Power(Base=@n25.G, Exp=@n33) {constExponent=10}))
  NamedRerouteDeclaration(None=Multiply(A=@n15, B=Add(A=@n31, B=@n34))) {name=BlockEmi, nodeColor=<rgba 1 0 0.1357 1>}
  NamedRerouteDeclaration(None=@n34) {name=BlockMask, nodeColor=<rgba 1 0 0.1357 1>}
  n35 = ComponentMask(None=TextureCoordinate()) {g=true}
  n36 = Power(Base=@n35, Exp=ScalarParameter() {defaultValue=1.7, group=SimpleMask, parameterName=SimpleMaskPower}) {constExponent=1.5}
  n37 = Saturate(None=OneMinus(None=@n35))
  NamedRerouteDeclaration(None=Saturate(None=Multiply(A=@n36, B=@n37))) {name=SimpleMask, nodeColor=<rgba 0.4531 1 0 1>}
  n38 = VertexColor()
  n39 = Multiply(A=@n38.R, B=VectorParameter() {group=HandHologram, parameterName=Outside Color1, defaultValue=<rgba 5 0 3.0833 1>}.RGB)
  n40 = Multiply(A=@n38.G, B=VectorParameter() {group=HandHologram, parameterName=Outside Color2, defaultValue=<rgba 0.2667 0 2 1>}.RGB)
  n41 = MaterialFunctionCall(Float2=TextureCoordinate()) {materialFunction=&BreakOutFloat2Components}
  n42 = MaterialFunctionCall(X=@n41.G, Y=@n41.G) {materialFunction=&MakeFloat2}
  n43 = Multiply(
   A=@n38.B
   B=TextureSample(UVs=Panner(Coordinate=@n42.Result) {speedX=1}) {texture=/BRCosmetics/Weapons/WeaponSkins/Textures/BadassMale_Heat_Gradient}.RGB
  )
  n44 = CurveAtlasRowParameter(CurveTime=MaterialFunctionCall(Float2=TextureCoordinate()) {materialFunction=&BreakOutFloat2Components}.G) {
   atlas=&VFX_CurveAtlas
   curve=/Game/Athena/LinearColorCurves/Curve_Contrail_GhostHunter_03
   defaultValue=35
   group=HandHologram
   parameterName=RingCurve
  }
  n45 = VectorParameter() {group=HandHologram, parameterName=RingColor1, defaultValue=<rgba 0.6333 0 1 1>}
  n46 = VectorParameter() {group=HandHologram, parameterName=RingColor2, defaultValue=<rgba 0 1 0.9333 1>}
  n47 = Add(
   A=Add(A=Multiply(A=@n44.R, B=@n45.RGB), B=Multiply(A=@n44.G, B=@n46.RGB))
   B=Multiply(A=@n44.B, B=VectorParameter() {group=HandHologram, parameterName=RingColor3, defaultValue=<rgba 0 10 1.3333 1>}.RGB)
  )
  NamedRerouteDeclaration(None=Add(A=Add(A=@n39, B=@n40), B=Multiply(A=@n43, B=@n47))) {name=HandHologramEmi, nodeColor=<rgba 0.2129 0 1 1>}
  n48 = Multiply(A=@n38.B, B=Add(A=Add(A=@n44.R, B=@n44.R), B=@n44.B))
  n49 = MaterialFunctionCall(Float2=TextureCoordinate()) {materialFunction=&BreakOutFloat2Components}
  n50 = Time()
  n51 = Frac(None=Add(A=Multiply(A=@n49.G) {constB=2}, B=Multiply(A=@n50) {constB=3}))
  n52 = Multiply(A=@n51, B=OneMinus(None=@n51))
  n53 = LinearInterpolate(Alpha=Ceil(None=Add(A=@n52) {constB=-0.2})) {constA=0.1}
  NamedRerouteDeclaration(None=Add(A=Add(A=@n38.R, B=@n38.R), B=Multiply(A=@n48, B=@n53))) {name=HandHologramMask, nodeColor=<rgba 0.2129 0 1 1>}
 ]
}
