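# GhostHunter skydiving contrail Niagara system, attached to two bone or socket points on the player's skeletal mesh.
# Cubes and small Plane meshes stream off those points, while burst-spawned LightRays and ParticleDecal add layered beams, glows, lights and a spinning hologram.
# System dive amount drives spawn, size, motion or opacity per emitter, with time-of-day and front-end dimming.
# Cubes send position in death events; Glitch sprites lack their own spawn module, suggesting event-driven spawning.
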NiagaraSystem /BRCosmetics/Athena/Effects/Particles/Skydiving/NS_Contrail_GhostHunter/NS_Contrail_GhostHunter
dependencies {
 usedDataInterfaces [&NiagaraDataInterfaceCurve, &NiagaraDataInterfaceArrayFloat3, &NiagaraDataInterfaceArrayFloat]
 usedDynamicInputs [
  &MakeCustomFloatFromBool
  &Divide_Float
  &ModuloFloat
 ]
 usedModules [
  &self:SystemSpawnScript.SystemScriptSource.SystemScriptGraph.NiagaraNodeAssignment_2.SetVariables_#1
  &self:SystemSpawnScript.SystemScriptSource.SystemScriptGraph.NiagaraNodeAssignment_3.SetVariables_#1
  /Game/Effects/Niagara/Modules/Skydiving/FN_DivingAttributes/v4/FN_DivingAttributes_v4
  &self:SystemSpawnScript.SystemScriptSource.SystemScriptGraph.NiagaraNodeAssignment_4.SetVariables_#2
  &SystemState
 ]
}
emitters {
 Cubes {
  emitterUpdateScript [
   &EmitterState(Inactive Response=::NewEnumerator1, Life Cycle Mode=::NewEnumerator1, Loop Duration=1, UseLoopDelay=false)
   /SetVariables(Emitter.Diving:NiagaraFloat=&Lerp_Float(A=0.5, B=1.2, Alpha=&ClampFloat(Float=$System.Dive))) {isSetParameters=true, module=SetVariables_#3}
   &SpawnRate(SpawnRate=&Multiply_Float(A=2, B=$Emitter.Diving), Use Spawn Probability=false)
  ]
  particleSpawnScript [
   &FN_SetSpawnGroup(Number of Spawn Groups=2)
   &InitializeParticle(
    Color Maximum=<rgba 1 1 1 0.5>
    Color Minimum=<rgba 0.6042 1 0.6174 2>
    Color Mode=::NewEnumerator2
    Lifetime Mode=::NewEnumerator1
    Mesh Scale Mode=::NewEnumerator1
    Mesh Uniform Scale Min=1.3
    UsePositionOffset=false
    Write Material Random=true
   )
   &InitialMeshOrientation(
    Orientation Axis=<0 0 1>
    Orientation Vector=<0 0 1>
    Rotation=&RandomRangeVector(Minimum=<-1 -1 -1>, Maximum=<1 1 1>)
    Use Orientation Vector=true
    Use Rotation Vector=true
   )
   &AddVelocity(Velocity=&Multiply_VectorByFloat(Vector=&RandomRangeVector(Minimum=<10 10 50>, Maximum=<-10 -10 100>), Float=$Emitter.Diving))
  ]
  particleUpdateScript [
   &ParticleState()
   &Drag(
    Drag=&Multiply_Float(A=&RandomRangeFloat(Minimum=0, Maximum=1), B=$Emitter.Diving)
    Use Linear Drag=true
    Use Rotational Drag=false
   )
   &SolveForcesAndVelocity(Clamp Velocity=false, Limit Acceleration=false)
   &SkeletalMeshLocation(
    Bone / Socket Index=$Particles.SpawnGroup
    Bone Sampling Mode=::NewEnumerator4
    MaskSpawnGroup=false
    Orientation Sampling=::NewEnumerator1
    Position Sampling=::NewEnumerator2
    Skeletal Mesh=$User.Skeletal Mesh
   )
   &UpdateMeshOrientation(
    Rotation Rate=&RandomRangeFloat(Minimum=0.2, Maximum=0.4)
    Rotation Vector=&RandomRangeVector(Minimum=<-1 -1 -1>, Maximum=<1 1 1>)
   )
   &ScaleColor(
    Scale Alpha=&FloatFromCurve(
     FloatCurve=DataInterface() {propertyValues {Curve {keys [<at 0.85 1>,<at 1 0>]}}}
     Scale Curve=&Multiply_Float(
      A=&FNTimeOfDayScalar(Night Scale=0.5, Morning Scale=0.3, Evening Scale=0.7)
      B=&ClampFloat(Float=$Emitter.Diving)
     )
    )
    ScaleA=true
    ScaleRGB=true
   )
   &ScaleMeshSize(Scale Factor=&VectorFromFloat(Value=&FloatFromCurve(FloatCurve=DataInterface() {propertyValues {Curve {keys [<at 0 0 RCIM_Cubic>, {arriveTangent=5.7471, interpMode=RCIM_Cubic, leaveTangent=5.7471, tangentMode=RCTM_User, time=0.3, value=1}]}}})))
   &ScaleMeshSize(Scale Factor=&VectorFromFloat(Value=&ClampFloat(Float=$Emitter.Diving, Min=0.8))) {module=ScaleMeshSize001}
   /SetVariables(Particles.NewPosition:Vector3f=&Add_Vector(
     A=Linked() {linkedVariable {name=Particles.Position, type=NiagaraPosition}}
     B=Linked() {linkedVariable {name=Particles.SkeletalMeshLocation.SampledPosition, type=NiagaraPosition}}
    )) {isSetParameters=true}
   &DynamicMaterialParameters(
    Index 0 Param 1=&RandomRangeFloat(Minimum=-0.15, Maximum=0.1)
    Index 0 Param 2=&RandomRangeFloat(Minimum=1, Maximum=1.5)
    Index 0 Param 3=&RandomRangeFloat(Minimum=-2, Maximum=-1)
    Module.Param0WriteEnabled=true
    Module.Param1WriteEnabled=true
    Module.Param2WriteEnabled=true
    Module.Param3WriteEnabled=true
   )
   &GenerateDeathEvent(Use Event Probability=false, "Vector to Send as Vector 1 (Position)":Vector3f=$Particles.NewPosition)
  ]
  rendererClasses [&NiagaraMeshRendererProperties]
  renderers [{rendererClass=&NiagaraMeshRendererProperties, rendererIndex=0}]
 }
 Glitch {
  emitterUpdateScript [
   &EmitterState(
    Inactive Response=::NewEnumerator1
    Life Cycle Mode=::NewEnumerator1
    Loop Duration=1
    Recalculate Duration Each Loop=true
    UseLoopDelay=false
   )
   /SetVariables(Emitter.Diving:NiagaraFloat=&Lerp_Float(A=0, B=1.2, Alpha=&ClampFloat(Float=$System.Dive))) {isSetParameters=true, module=SetVariables_#3}
  ]
  particleSpawnScript [
   &InitializeParticle(
    Color=<rgba 0 0 0 1>
    Lifetime Max=0.5
    Lifetime Min=0.3
    Lifetime Mode=::NewEnumerator1
    Sprite Size Max=<40 30>
    Sprite Size Min=<20 15>
    Sprite Size Mode=::NewEnumerator2
    UsePositionOffset=false
    Write Material Random=true
   ) {module=InitializeParticle001}
   &CameraOffset(Camera Offset Amount=&RandomRangeFloat(Minimum=0, Maximum=35))
   &DynamicMaterialParameters(
    Index 0 Param 1=&RandomRangeFloat(Minimum=0, Maximum=1)
    Index 0 Param 2=&RandomRangeFloat(Minimum=0, Maximum=0.5)
    Index 0 Param 3=-3
    Index 0 Param 4=5
    Module.Param0WriteEnabled=true
    Module.Param1WriteEnabled=true
    Module.Param2WriteEnabled=true
    Module.Param3WriteEnabled=true
   )
  ]
  particleUpdateScript [
   &ParticleState()
   &ScaleSpriteSize(Scale Factor=&MakeVector2D(
     X=&FloatFromCurve(FloatCurve=DataInterface() {propertyValues {Curve {keys [<at 0 1.5>,<at 1 0.5>]}}})
     Y=&FloatFromCurve(FloatCurve=DataInterface() {propertyValues {Curve {keys [<at 0 1>,<at 1 2>]}}})
    ))
   &SolveForcesAndVelocity(Clamp Velocity=false, Limit Acceleration=false)
   &ScaleSpriteSize(Scale Sprite Size Mode=::NewEnumerator0, Uniform Scale Factor=&Multiply_Float(A=1, B=$Emitter.Diving)) {module=ScaleSpriteSize001}
  ]
  rendererClasses [&NiagaraSpriteRendererProperties]
  renderers [{rendererClass=&NiagaraSpriteRendererProperties}]
 }
 LightRays {
  emitterUpdateScript [
   &EmitterState(
    Inactive Response=::NewEnumerator1
    Life Cycle Mode=::NewEnumerator1
    Loop Behavior=::NewEnumerator1
    Loop Duration Mode=::NewEnumerator1
    UseLoopDelay=false
   )
   /SetVariables(Emitter.Diving:NiagaraFloat=&Lerp_Float(A=0.7, B=1.1, Alpha=&ClampFloat(Float=$System.Dive))) {isSetParameters=true}
   /SetVariables(Emitter.DivingAlpha:NiagaraFloat=&Lerp_Float(A=0.5, B=1.1, Alpha=&ClampFloat(Float=$System.Dive))) {isSetParameters=true}
   &SpawnBurst_Instantaneous(Spawn Count=2, Use Spawn Probability=false)
  ]
  particleSpawnScript [
   &FN_SetSpawnGroup(Number of Spawn Groups=2) {module=FN_SetSpawnGroup001}
   &InitializeParticle(Color=<rgba 2 2 2 2>, Mesh Scale=<0.55 0.37 0.37>, Mesh Scale Mode=::NewEnumerator4, UsePositionOffset=false, Write Material Random=true)
   &InitialMeshOrientation(Rotation=<0 -0.25 0>, Use Orientation Vector=true, Use Rotation Vector=true) {module=InitialMeshOrientation001}
   &InitialMeshOrientation(
    Rotation=&RandomRangeVector(Minimum=<0 0 1>, Maximum=<0 0 0>)
    Use Orientation Vector=true
    Use Rotation Vector=true
   )
   &CameraOffset(Camera Offset Amount=20)
  ]
  particleUpdateScript [
   &ParticleState(Kill Particles When Lifetime Has Elapsed=false, Let Infinitely Lived Particles Die When Emitter Deactivates=true)
   &SolveForcesAndVelocity(Clamp Velocity=false, Limit Acceleration=false)
   &SkeletalMeshLocation(
    Bone / Socket Index=$Particles.SpawnGroup
    Bone Sampling Mode=::NewEnumerator4
    MaskSpawnGroup=false
    Orientation Sampling=::NewEnumerator1
    Sampled Position Offset=&SelectVectorFromArray(Direct Array Index=$Particles.SpawnGroup, Vector Selection Array=DataInterface() {propertyValues {FloatData [<-5 0 0>, <5 0 0>]}})
    Skeletal Mesh=$User.Skeletal Mesh
   )
   &ScaleColor(Scale Alpha=&Multiply_Float(A=$System.TOD, B=$Emitter.DivingAlpha), ScaleA=true, ScaleRGB=true)
   &ScaleMeshSize(Scale Factor=&Multiply_VectorByFloat(Vector=<1 1 1>, Float=$Emitter.Diving))
   /SetVariables(
    Particles.DarkColor:LinearColor=&MultiplyLinearColor(
     Linear Color=<rgba 1 1 1 2>
     Scale Factor=&MakeFloat4FromVectorAndFloat("Vector (RGB)"=<1 1 1>, "Float (Alpha)"=&Multiply_Float(A=$System.TOD, B=$Emitter.DivingAlpha))
    )
    Particles.DarkSize:Vector3f=&Multiply_VectorByFloat(Vector=<0.6 0.5 0.5>, Float=$Emitter.Diving)
   ) {isSetParameters=true}
   /SetVariables(
    Particles.BottomColor:LinearColor=&MultiplyLinearColor(
     Linear Color=<rgba 2 2 2 4>
     Scale Factor=&MakeFloat4FromVectorAndFloat("Vector (RGB)"=<1 1 1>, "Float (Alpha)"=&Multiply_Float(A=$System.TOD, B=$Emitter.DivingAlpha))
    )
    Particles.BottomSize:Vector3f=&Multiply_VectorByFloat(Vector=<0.2 0.35 0.35>, Float=$Emitter.Diving)
   ) {isSetParameters=true}
   /SetVariables(
    Particles.BlockColor:LinearColor=&MultiplyLinearColor(
     Linear Color=<rgba 1 1 1 1>
     Scale Factor=&MakeFloat4FromVectorAndFloat(
      "Vector (RGB)"=<1 1 1>
      "Float (Alpha)"=&Multiply_Float(
       A=&Multiply_Float(A=$System.TOD, B=$Emitter.DivingAlpha)
       B=&FN_FrontEndFloatModifier(Gameplay=1, Front End=0.2)
      )
     )
    )
    Particles.BlockSize:Vector3f=&Multiply_VectorByFloat(Vector=<0.4 0.3 0.3>, Float=$Emitter.Diving)
   ) {isSetParameters=true}
   /SetVariables(
    Particles.Dark2Color:LinearColor=&MultiplyLinearColor(
     Linear Color=<rgba 1 1 1 0.5>
     Scale Factor=&MakeFloat4FromVectorAndFloat(
      "Vector (RGB)"=<1 1 1>
      "Float (Alpha)"=&Multiply_Float(
       A=&Multiply_Float(A=$System.TOD, B=$Emitter.DivingAlpha)
       B=&FN_FrontEndFloatModifier(Gameplay=1, Front End=0.5)
      )
     )
    )
    Particles.Dark2Size:Vector3f=&Multiply_VectorByFloat(Vector=<0.6 0.5 0.5>, Float=$Emitter.Diving)
   ) {isSetParameters=true}
   /SetVariables(
    Particles.LightRay2Color:LinearColor=&MultiplyLinearColor(
     Linear Color=<rgba 1 1 1 0.3>
     Scale Factor=&MakeFloat4FromVectorAndFloat("Vector (RGB)"=<1 1 1>, "Float (Alpha)"=&Multiply_Float(A=$System.TOD, B=$Emitter.DivingAlpha))
    )
    Particles.LightRay2Size:Vector3f=&Multiply_VectorByFloat(Vector=<1 0.4 0.4>, Float=$Emitter.Diving)
   ) {isSetParameters=true}
   /SetVariables(
    Particles.BottomDarkColor:LinearColor=&MultiplyLinearColor(
     Linear Color=<rgba 0.0651 0 0.3906 5>
     Scale Factor=&MakeFloat4FromVectorAndFloat(
      "Vector (RGB)"=<1 1 1>
      "Float (Alpha)"=&Multiply_Float(
       A=&Multiply_Float(A=$System.TOD, B=$Emitter.DivingAlpha)
       B=&FN_FrontEndFloatModifier(Gameplay=1, Front End=0.5)
      )
     )
    )
    Particles.BottomDarkSize:Vector3f=&Multiply_VectorByFloat(Vector=<0.2 0.4 0.4>, Float=$Emitter.Diving)
   ) {isSetParameters=true}
   /SetVariables(
    Particles.Dark3Color:LinearColor=&MultiplyLinearColor(
     Linear Color=<rgba 1 1 1 0.2>
     Scale Factor=&MakeFloat4FromVectorAndFloat(
      "Vector (RGB)"=<1 1 1>
      "Float (Alpha)"=&Multiply_Float(
       A=&Multiply_Float(A=$System.TOD, B=$Emitter.DivingAlpha)
       B=&FN_FrontEndFloatModifier(Gameplay=1, Front End=0.5)
      )
     )
    )
    Particles.Dark3Size:Vector3f=&Multiply_VectorByFloat(Vector=<1 0.4 0.4>, Float=$Emitter.Diving)
   ) {isSetParameters=true}
   /SetVariables(Particles.PositionOffset:Vector3f=&Add_Vector(A=Linked() {linkedVariable {name=Particles.Position, type=NiagaraPosition}}, B=<0 0 -5>)) {isSetParameters=true, module=SetVariables_#4}
  ]
  rendererClasses [&NiagaraMeshRendererProperties]
  renderers [
   {rendererClass=&NiagaraMeshRendererProperties, rendererIndex=0}
   {rendererClass=&NiagaraMeshRendererProperties, rendererIndex=1}
   {rendererClass=&NiagaraMeshRendererProperties, rendererIndex=2}
   {rendererClass=&NiagaraMeshRendererProperties, rendererIndex=3}
   {rendererClass=&NiagaraMeshRendererProperties, rendererIndex=4}
   {rendererClass=&NiagaraMeshRendererProperties, rendererIndex=5}
   {rendererClass=&NiagaraMeshRendererProperties, rendererIndex=6}
   {rendererClass=&NiagaraMeshRendererProperties, rendererIndex=7}
  ]
 }
 ParticleDecal {
  emitterSpawnScript [/SetVariables(Emitter.Use Hand Hologram=1) {isSetParameters=true}]
  emitterUpdateScript [
   &EmitterState(Life Cycle Mode=::NewEnumerator1, Loop Behavior=::NewEnumerator1, Loop Duration Mode=::NewEnumerator1, UseLoopDelay=false)
   /SetVariables(Emitter.Diving:NiagaraFloat=&Lerp_Float(A=0.5, B=1, Alpha=&ClampFloat(Float=$System.Dive))) {isSetParameters=true, module=SetVariables_#3}
   /SetVariables(Emitter.DivingSize:NiagaraFloat=&Lerp_Float(A=0.8, B=1.1, Alpha=&ClampFloat(Float=$System.Dive))) {isSetParameters=true}
   /SetVariables(Emitter.DivingColor:NiagaraFloat=&Lerp_Float(
     A=0.5
     B=&FN_FrontEndFloatModifier(Gameplay=2, Front End=1.5)
     Alpha=&ClampFloat(Float=$System.Dive)
    )) {isSetParameters=true}
   &SpawnBurst_Instantaneous(Spawn Count=9, Use Spawn Probability=false)
  ]
  particleSpawnScript [
   &FN_SetSpawnGroup(Number of Spawn Groups=2)
   &InitializeParticle(Color=<rgba 0 1 0.3167 0.1>, Sprite Size Mode=::NewEnumerator3, Uniform Sprite Size=50, UsePositionOffset=false, Write Material Random=true)
   &CameraOffset()
   &DynamicMaterialParameters(
    Index 0 Param 1=1
    Index 0 Param 2=1
    Module.Param0WriteEnabled=true
    Module.Param1WriteEnabled=true
    Module.Param2WriteEnabled=true
    Module.Param3WriteEnabled=true
   )
   /SetVariables(Particles.DecalCamera offset=20) {isSetParameters=true}
  ]
  particleUpdateScript [
   &ParticleState(Kill Particles When Lifetime Has Elapsed=false, Let Infinitely Lived Particles Die When Emitter Deactivates=true)
   &SkeletalMeshLocation(
    Bone / Socket Index=$Particles.SpawnGroup
    Bone Sampling Mode=::NewEnumerator4
    MaskSpawnGroup=false
    Orientation Sampling=::NewEnumerator1
    Sampled Position Offset=&SelectVectorFromArray(Direct Array Index=$Particles.SpawnGroup, Vector Selection Array=DataInterface() {propertyValues {FloatData [<-5 0 0>, <5 0 0>]}})
    Skeletal Mesh=$User.Skeletal Mesh
   )
   &ScaleColor(ScaleA=true, ScaleRGB=true)
   /SetVariables(
    Particles.Glow01Color:LinearColor=&MultiplyLinearColor(
     Linear Color=<rgba 0 1 0.3833 1>
     Scale Factor=&MakeFloat4FromVectorAndFloat(
      "Vector (RGB)"=<1 1 1>
      "Float (Alpha)"=&Multiply_Float(
       A=&Multiply_Float(A=$System.TOD, B=$Emitter.Diving)
       B=&FN_FrontEndFloatModifier(Gameplay=1, Front End=0.5)
      )
     )
    )
    Particles.Glow01Size:Vector2f=&Multiply_Vector2D_ByFloat(Float=$Emitter.Diving, Vector2D=<15 15>)
   ) {isSetParameters=true}
   /SetVariables(
    Particles.Glow02Color:LinearColor=&MultiplyLinearColor(
     Linear Color=<rgba 0 1 0.65 0.01>
     Scale Factor=&MakeFloat4FromVectorAndFloat(
      "Vector (RGB)"=<1 1 1>
      "Float (Alpha)"=&Multiply_Float(
       A=&Multiply_Float(A=$System.TOD, B=1)
       B=&FN_FrontEndFloatModifier(Gameplay=1, Front End=2)
      )
     )
    )
    Particles.Glow02Size:Vector2f=&Multiply_Vector2D_ByFloat(Float=$Emitter.Diving, Vector2D=<70 70>)
   ) {isSetParameters=true}
   /SetVariables(Particles.LightColor=<rgba 0.0562 5 3.5993 1>, Particles.LightFallOff=2, Particles.LightRadius=40) {isSetParameters=true}
   /SetVariables(
    Particles.HologramColor:LinearColor=&MultiplyLinearColor(
     Linear Color=&MakeLinearColorFromVectorAndFloat(
      "Vector (RGB)"=&VectorFromFloat(Value=&FN_FrontEndFloatModifier(Gameplay=1, Front End=0.1))
      "Float (Alpha)"=1
     )
     Scale Factor=&MakeFloat4FromVectorAndFloat(
      "Vector (RGB)"=&VectorFromFloat(Value=$Emitter.DivingColor)
      "Float (Alpha)"=&Multiply_Float(A=$System.TOD, B=$Emitter.Diving)
     )
    )
    Particles.HologramSize:Vector3f=&VectorFromFloat(Value=&Multiply_Float(A=1, B=$Emitter.DivingSize))
   ) {isSetParameters=true}
   /SetVariables(Particles.PositionOffset:Vector3f=&Add_Vector(A=Linked() {linkedVariable {name=Particles.Position, type=NiagaraPosition}}, B=<0 0 -5>)) {isSetParameters=true, module=SetVariables_#4}
   /SetVariables(
    Particles.Glow03Color:LinearColor=&MultiplyLinearColor(
     Linear Color=<rgba 0.2167 0 1 0.35>
     Scale Factor=&MakeFloat4FromVectorAndFloat(
      "Vector (RGB)"=<1 1 1>
      "Float (Alpha)"=&Multiply_Float(
       A=&Multiply_Float(A=$System.TOD, B=$Emitter.Diving)
       B=&FN_FrontEndFloatModifier(Gameplay=1, Front End=2)
      )
     )
    )
    Particles.Glow03Facing:Vector3f=<0 0 1>
    Particles.Glow03Size:Vector2f=&Multiply_Vector2D_ByFloat(Float=$Emitter.DivingSize, Vector2D=<60 60>)
   ) {isSetParameters=true}
   &UpdateMeshOrientation(
    Rotation Coordinate Space=::NewEnumerator2
    Rotation Rate=&Multiply_Float(
     A=&SelectFloatFromArray(Direct Array Index=$Particles.SpawnGroup, Float Selection Array=DataInterface() {propertyValues {FloatData [0.35,-0.35]}})
     B=$Emitter.DivingColor
    )
    Rotation Vector=<0 0 1>
   )
  ]
  rendererClasses [
   &NiagaraSpriteRendererProperties
   &NiagaraLightRendererProperties
   &NiagaraMeshRendererProperties
  ]
  renderers [
   {rendererClass=&NiagaraSpriteRendererProperties}
   {rendererClass=&NiagaraSpriteRendererProperties, rendererIndex=1}
   {rendererClass=&NiagaraSpriteRendererProperties, rendererIndex=2}
   {rendererClass=&NiagaraLightRendererProperties, rendererIndex=3}
   {rendererClass=&NiagaraMeshRendererProperties, rendererIndex=4}
   {rendererClass=&NiagaraSpriteRendererProperties, rendererIndex=5}
  ]
 }
 Plane {
  emitterUpdateScript [
   &EmitterState(Inactive Response=::NewEnumerator1, Life Cycle Mode=::NewEnumerator1, Loop Duration=1, UseLoopDelay=false)
   /SetVariables(Emitter.Diving:NiagaraFloat=&Lerp_Float(A=0.1, B=1.2, Alpha=&ClampFloat(Float=$System.Dive))) {isSetParameters=true, module=SetVariables_#3}
   &SpawnRate(SpawnRate=&Multiply_Float(A=20, B=$Emitter.Diving), Use Spawn Probability=false)
  ]
  particleSpawnScript [
   &FN_SetSpawnGroup(Number of Spawn Groups=2) {module=FN_SetSpawnGroup001}
   &InitializeParticle(
    Color Maximum=&RandomRangeLinearColor(Minimum=<rgba 0.4167 0 1 2>, Maximum=<rgba 0.95 0 1 2>)
    Color Minimum=&RandomRangeLinearColor(Minimum=<rgba 0.0667 1 0 2>, Maximum=<rgba 0 0.3833 1 2>)
    Color Mode=::NewEnumerator2
    Lifetime Max=1.6
    Lifetime Min=0.7
    Lifetime Mode=::NewEnumerator1
    Mesh Scale Max=<0.08 0.08 0.08>
    Mesh Scale Min=<0.03 0.03 0.03>
    Mesh Scale Mode=::NewEnumerator2
    UsePositionOffset=false
    Write Material Random=true
   ) {module=InitializeParticle001}
   &CylinderLocation(
    Cylinder Height=1
    Cylinder Radius=&RandomRangeFloat(Minimum=10, Maximum=20)
    MaskSpawnGroup=false
    Override Seed=false
    UseOwnerScale=true
   )
   &AddVelocity(Velocity=&Multiply_VectorByFloat(Vector=&RandomRangeVector(Minimum=<30 30 350>, Maximum=<-30 -30 100>), Float=$Emitter.Diving))
   &InitialMeshOrientation(
    Rotation=&RandomRangeVector(Minimum=<-1 -1 -1>, Maximum=<1 1 1>)
    Use Orientation Vector=true
    Use Rotation Vector=true
   )
   &DynamicMaterialParameters(
    Index 0 Param 1=&RandomRangeFloat(Minimum=0, Maximum=1)
    Index 0 Param 2=&UniformAOrBFloat(A=0)
    Index 0 Param 3=&UniformAOrBFloat(A=0, B=0.5)
    Index 0 Param 4=&RandomRangeFloat(Minimum=0.5, Maximum=1)
    Module.Param0WriteEnabled=true
    Module.Param1WriteEnabled=true
    Module.Param2WriteEnabled=true
    Module.Param3WriteEnabled=true
   )
  ]
  particleUpdateScript [
   &ParticleState()
   &Drag(Drag=&Multiply_Float(A=2, B=$Emitter.Diving), Use Linear Drag=true, Use Rotational Drag=false)
   &SolveForcesAndVelocity(Clamp Velocity=false, Limit Acceleration=false)
   &SkeletalMeshLocation(
    Bone / Socket Index=$Particles.SpawnGroup
    Bone Sampling Mode=::NewEnumerator4
    MaskSpawnGroup=false
    Orientation Sampling=::NewEnumerator1
    Position Sampling=::NewEnumerator2
    Skeletal Mesh=$User.Skeletal Mesh
   )
   &UpdateMeshOrientation(
    Rotation Rate=&RandomRangeFloat(Minimum=1, Maximum=1)
    Rotation Vector=&RandomRangeVector(Minimum=<-1 -1 -1>, Maximum=<1 1 1>)
   )
   /SetVariables(Particles.NewPosition:Vector3f=&Add_Vector(
     A=Linked() {linkedVariable {name=Particles.Position, type=NiagaraPosition}}
     B=Linked() {linkedVariable {name=Particles.SkeletalMeshLocation.SampledPosition, type=NiagaraPosition}}
    )) {isSetParameters=true}
   &ScaleMeshSize(Scale Factor=&VectorFromFloat(Value=&FloatFromCurve(FloatCurve=DataInterface() {propertyValues {Curve {keys [<at 0 0>,<at 0.1706 1>,<at 0.7 1>,<at 1 0>]}}})))
   &ScaleColor(Scale Alpha=&RandomRangeFloat(Minimum=0.5, Maximum=1), ScaleA=true, ScaleRGB=true)
   &ScaleColor(
    Scale Alpha=&FNTimeOfDayScalar(Night Scale=0.5, Morning Scale=0.3, Evening Scale=0.7)
    ScaleA=true
    ScaleRGB=true
   ) {module=ScaleColor001}
  ]
  rendererClasses [&NiagaraMeshRendererProperties]
  renderers [{rendererClass=&NiagaraMeshRendererProperties, rendererIndex=0}]
 }
}
summary {
 systemName=NS_Contrail_GhostHunter
 emitters [
  {emitterName=Cubes, rendererClasses [&NiagaraMeshRendererProperties]}
  {emitterName=Plane, rendererClasses [&NiagaraMeshRendererProperties]}
  {emitterName=LightRays, rendererClasses [&NiagaraMeshRendererProperties]}
  {
   emitterName=ParticleDecal
   rendererClasses [
    &NiagaraSpriteRendererProperties
    &NiagaraLightRendererProperties
    &NiagaraMeshRendererProperties
   ]
  }
  {emitterName=Glitch, rendererClasses [&NiagaraSpriteRendererProperties]}
 ]
 userVariables [
  User.bIsFrontEnd {type=NiagaraBool}
  User.bIsFrontEndPreview {type=NiagaraBool}
  User.Skeletal Mesh {
   type=NiagaraDataInterfaceSkeletalMesh
   defaultValue {
    struct=NiagaraExt_VariableValue_DataInterface
    value {dataInterface=&self:NiagaraDataInterfaceSkeletalMesh_0}
   }
  }
 ]
}
